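# Unlit, translucent, two-sided Niagara ribbon material for the paint trail in the Dazzle emote.
# Tiled Gaussian noise shades the particle color for emissive and feeds an erosion function driven by a sphere mask and particle alpha.
# Opacity adds a masked Fresnel glow, constant on mobile, to that erosion, with an optional soft front edge.
# Dynamic parameters drive global opacity and optional glow strength, and depth fade is skipped at Low quality.
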Material /BRCosmetics/Animation/Game/MainPlayer/Emotes/Dazzle/FX/M_Dazzle_Emote_PaintTrail
graph {
 properties {
  bCastRayTracedShadows=false
  bUsedWithNiagaraRibbons=true
  blendMode=BLEND_Translucent
  opacityMaskClipValue=0.01
  pixelDepthOffsetMode=PDOM_Legacy
  shadingModel=MSM_Unlit
  twoSided=true
  usedShadingModels=Unlit
 }
 nodes [
  n1 = TextureSample(UVs=LinearInterpolate(
    A=TextureCoordinate() {coordinateIndex=1, vTiling=2}
    B=TextureCoordinate() {coordinateIndex=1, vTiling=5}
    Alpha=ScalarParameter() {parameterName=UseSoftFrontEdge}
   )) {texture=/Game/Effects/Fort_Effects/Textures/Tile/T_GaussianNoise_01_M}
  n2 = Power(Base=@n1.R) {constExponent=0.25}
  n3 = Add(A=Saturate(None=OneMinus(None=@n2))) {constB=0.5}
  n4 = ParticleColor()
  MaterialFunctionCall(Emissive=Saturate(None=Multiply(A=@n3, B=@n4.RGB))) {materialFunction=&MF_FX_EmissiveAdjustments}
  n5 = Multiply(
   A=Power(Base=@n1.R, Exp=ScalarParameter() {defaultValue=2, parameterName=Contrast}) {constExponent=1.25}
   B=ScalarParameter() {defaultValue=4, parameterName=Strenght}
  ) {constB=2}
  n6 = Multiply(
   A=SphereMask(A=TextureCoordinate() {uTiling=0}, B=Constant2Vector() {g=0.5}, Radius=Constant() {r=0.35}, Hardness=Constant() {r=0.25})
   B=@n4.A
  ) {constB=0.5}
  n7 = ScalarParameter() {parameterName=Softness}
  n8 = ComponentMask(None=TextureCoordinate()) {r=true}
  n9 = TextureCoordinate() {uTiling=0}
  n10 = Constant2Vector() {g=0.5}
  n11 = Constant() {r=0.5}
  n12 = Constant() {r=0.1}
  n13 = Saturate(None=Multiply(A=OneMinus(None=@n8), B=SphereMask(A=@n9, B=@n10, Radius=@n11, Hardness=@n12)))
  n14 = Multiply(A=Power(Base=@n8, Exp=ScalarParameter() {defaultValue=1, parameterName=SoftEdgePower}) {constExponent=0.5}, B=@n13)
  n15 = ScalarParameter() {parameterName=UseSoftFrontEdge}
  n16 = LinearInterpolate(A=@n13, B=@n14, Alpha=@n15)
  n17 = LinearInterpolate(
   A=ScalarParameter() {defaultValue=0.75, parameterName=GlowStrenght}
   B=DynamicParameter() {defaultValue=<rgba 1 0 1 1>, paramNames [GlobalOpacity, GlowStrenght, Param3, Param4]}.GlowStrenght
   Alpha=ScalarParameter() {parameterName=UseParticleGlow}
  )
  n18 = MaterialFunctionCall(Brightness=Constant() {r=1}, Exponent=Constant() {r=2}, Vertex / Pixel Normal=VertexNormalWS()) {materialFunction=&MF_Fresnel_FX}
  n19 = Constant() {r=1}
  n20 = Add(
   A=OneMinus(None=MaterialFunctionCall!(Gradient=@n5, Erosion Amount=@n6, Softness=@n7) {materialFunction=/BRCosmetics/Weapons/FORT_Melee/Pickaxe_Turtleneck_Male/FX/Default/MF_Erosion}.Increased Range)
   B=Multiply(
    A=Multiply(A=@n16, B=@n17)
    B=ShadingPathSwitch(Default=@n18.Fresnel, Deferred=@n18.Fresnel, Mobile=@n19) {desc=Migrated from FeatureLevelSwitch}
   )
  )
  n21 = Multiply(A=@n20, B=@n14)
  n22 = Saturate(None=Multiply(
    A=LinearInterpolate(A=@n20, B=@n21, Alpha=@n15)
    B=DynamicParameter() {defaultValue=<rgba 1 0 1 1>, paramNames [GlobalOpacity, GlowStrenght, Param3, Param4]}.GlobalOpacity
   ))
  n23 = DepthFade(Opacity=@n22, FadeDistance=ScalarParameter() {parameterName=DepthFade})
  QualitySwitch(Default=@n23, Low=@n22, High=@n23, Medium=@n23, Epic=@n23)
 ]
}
