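# Niagara system for the Dazzle emote that draws colored paint trails from two bones or sockets on a skeletal mesh.
# Left and right ribbons trail those points, joined by glow sprites pinned to them and GPU embers spawned by sampled movement.
# Colors index a user brush color array by loop count, advancing each loop.
# Normalized loop age gates ember spawning, fades glow and ribbons, and kills them at loop end.
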NiagaraSystem /BRCosmetics/Animation/Game/MainPlayer/Emotes/Dazzle/FX/NS_Emote_Dazzle_Trail
dependencies {
 usedDataInterfaces [&NiagaraDataInterfaceCurve]
 usedDynamicInputs []
 usedModules [&SystemState, &SampleSkeletalMesh]
}
emitters {
 Embers {
  simTarget=GPUComputeSim
  emitterUpdateScript [
   &EmitterState()
   &SpawnPerUnit(
    Max Movement Threshold=500
    Module.Use Max Movement Threshold=true
    Module.Use Movement Tolerance=false
    Spawn Probability:NiagaraFloat=&FloatFromCurve(
     FloatCurve=DataInterface() {propertyValues {Curve {keys [
         {interpMode=RCIM_Cubic, tangentMode=RCTM_User, time=0, value=0}
         <at 0.05 1>
         {interpMode=RCIM_Cubic, tangentMode=RCTM_User, time=0.8, value=1}
         {tangentMode=RCTM_User, time=0.95, value=0}
        ]}}}
     CurveIndex=$System.NormalizedLoopAge
    )
    Spawn Spacing:NiagaraFloat=&RandomRangeFloat(Minimum=6, Maximum=10)
    Use Spawn Probability=true
    Velocity Vector:Vector3f=$System.SampleSkeletalMesh.SampledVelocity
   )
  ]
  particleSpawnScript [
   &InitializeParticle(
    Color=&SelectLinearColorFromArray(
     Array Sampling Mode=::NewEnumerator1
     Direct Array Index=$System.LoopCount
     Direct Array Mode=::NewEnumerator1
     Color Selection Array=$User.BrushColorArray
     Random Seed=?
    )
    Lifetime Max=1
    Lifetime Min=0.5
    Lifetime Mode=::NewEnumerator1
    Sprite Size=&Multiply_Vector2D_ByFloat(Float=&RandomRangeFloat(Minimum=0.5, Maximum=0.7), Vector2D=<1 1>)
    Sprite Size Mode=::NewEnumerator4
    UsePositionOffset=false
    Write Material Random=true
   )
   &PartitionParticles()
   &SkeletalMeshLocation(
    Bone / Socket Index=$Particles.PartitionParticles.Partition
    Bone Sampling Mode=::NewEnumerator4
    Interpolate Mesh Sampling=false
    MaskSpawnGroup=false
    Orientation Sampling=::NewEnumerator1
    Sampled Velocity Coordinate Space:ENiagaraCoordinateSpace=::Local
    Skeletal Mesh=$User.SkeletalMesh
   )
   &ShapeLocation(MaskSpawnGroup=false, Override Seed=false, Sphere Radius=7, UseOwnerScale=true)
   &InheritVelocity(
    Inherited Velocity Speed Limit:NiagaraFloat=&RandomRangeFloat(Minimum=50, Maximum=100)
    Speed Limit=true
    UseSpeedThreshold=false
    Velocity Source=$System.SampleSkeletalMesh.SampledVelocity
   )
   &ApplyInitialForces()
  ]
  particleUpdateScript [
   &ParticleState()
   &ScaleColor(
    Scale RGB=&Multiply_VectorByFloat(
     Vector=<1 1 1>
     Float=&FloatFromCurve(FloatCurve=DataInterface() {propertyValues {Curve {keys [{interpMode=RCIM_Cubic, tangentMode=RCTM_User, time=0, value=1}, {interpMode=RCIM_Cubic, tangentMode=RCTM_User, time=0.6, value=0.3}]}}})
    )
    ScaleA=true
    ScaleRGB=true
   )
   &ScaleSpriteSize(Uniform Curve Sprite Scale=DataInterface() {propertyValues {Curve {keys [
        {tangentMode=RCTM_User, time=0, value=1}
        {arriveTangent=-0.0405, interpMode=RCIM_Cubic, leaveTangent=-0.0405, tangentMode=RCTM_User, time=0.7009, value=1}
        {arriveTangent=-8.0089, interpMode=RCIM_Cubic, leaveTangent=-8.0089, tangentMode=RCTM_User, time=1, value=0}
       ]}}})
   &Drag(Drag=4, Use Linear Drag=true, Use Rotational Drag=true)
   &SolveForcesAndVelocity(Clamp Velocity=false, Limit Acceleration=false)
   &ScaleSpriteSizeBySpeed(Max Scale Factor=<1 4>, Velocity Threshold=35)
  ]
  rendererClasses [&NiagaraSpriteRendererProperties]
  renderers [{rendererClass=&NiagaraSpriteRendererProperties}]
 }
 Glow {
  emitterUpdateScript [
   &EmitterState()
   &SpawnBurst_Instantaneous(Spawn Count=2, Use Spawn Probability=false, UseLoopCountLimit=false)
  ]
  particleSpawnScript [
   &InitializeParticle(
    Color=&SelectLinearColorFromArray(
     Array Sampling Mode=::NewEnumerator1
     Direct Array Index=$System.LoopCount
     Direct Array Mode=::NewEnumerator1
     Color Selection Array=$User.BrushColorArray
     Random Seed=?
    )
    Sprite Size Mode=::NewEnumerator3
    Uniform Sprite Size=30
    UsePositionOffset=false
    Write Material Random=true
   )
   &PartitionParticles()
  ]
  particleUpdateScript [
   &SkeletalMeshLocation(
    Bone / Socket Index=$Particles.PartitionParticles.Partition
    Bone Sampling Mode=::NewEnumerator4
    MaskSpawnGroup=false
    Orientation Sampling=::NewEnumerator1
    Sampled Velocity Coordinate Space:ENiagaraCoordinateSpace=::Local
    Skeletal Mesh=$User.SkeletalMesh
   )
   &KillParticles(Kill Particles=&SetBoolByFloatComparison(A=0.99, B=$System.NormalizedLoopAge, Comparison Type=::NewEnumerator7))
   &ParticleState(Kill Particles When Lifetime Has Elapsed=false)
   &ScaleColor(
    Scale Alpha=&FloatFromCurve(
     FloatCurve=DataInterface() {propertyValues {Curve {keys [<at 0 0 RCIM_Cubic>, <at 0.1 1>, {arriveTangent=-1.25, interpMode=RCIM_Cubic, leaveTangent=-1.25, time=0.75, value=1}, <at 0.9 0 RCIM_Cubic>]}}}
     CurveIndex=$System.NormalizedLoopAge
     Scale Curve=0.2
    )
    Scale RGB=&Multiply_VectorByFloat(Vector=<1 1 1>, Float=0.5)
    ScaleA=true
    ScaleRGB=true
   )
   &FN_Light_Attributes(
    Color=$Particles.Color
    Create Color Attribute=true
    Create Exponent Attribute=true
    Create Position Attribute=true
    Create Radius Attribute=true
    Exponent=2
    Position=$Particles.SkeletalMeshLocation.SampledPosition
    Radius=75
    Scale Color=&Multiply_Float(
     A=4
     B=&FloatFromCurve(
      FloatCurve=DataInterface() {propertyValues {Curve {keys [<at 0 0 RCIM_Cubic>, <at 0.1 1>, {arriveTangent=-1.25, interpMode=RCIM_Cubic, leaveTangent=-1.25, time=0.75, value=1}, <at 0.9 0 RCIM_Cubic>]}}}
      CurveIndex=$System.NormalizedLoopAge
     )
    )
   )
  ]
  rendererClasses [&NiagaraSpriteRendererProperties]
  renderers [{rendererClass=&NiagaraSpriteRendererProperties}]
 }
 PaintTrail_L {
  emitterUpdateScript [
   &EmitterState()
   &SpawnRate(SpawnRate=22, Use Spawn Probability=false) {module=SpawnRate001}
  ]
  particleSpawnScript [
   &InitializeParticle(
    Color=&SelectLinearColorFromArray(
     Array Sampling Mode=::NewEnumerator1
     Direct Array Index=$System.LoopCount
     Direct Array Mode=::NewEnumerator1
     Color Selection Array=$User.BrushColorArray
     Random Seed=?
    )
    Lifetime=$User.Lifetime
    Position Mode=::NewEnumerator0
    Ribbon Width=$User.RibbonWidth
    Ribbon Width Mode=::NewEnumerator1
    Write Material Random=true
   )
   &SkeletalMeshLocation(
    Bone Sampling Mode=::NewEnumerator4
    Interpolate Mesh Sampling=false
    MaskSpawnGroup=false
    Orientation Sampling=::NewEnumerator1
    Sampled Position Offset Coordinate Space:ENiagaraSkelMeshTransforms=::NewEnumerator0
    Sampled Velocity Coordinate Space:ENiagaraCoordinateSpace=::Local
    Skeletal Mesh=$User.SkeletalMesh
    Use Sampled Mesh Attribute as Ribbon ID=false
   ) {module=SkeletalMeshLocation001}
  ]
  particleUpdateScript [
   &ParticleState()
   &ScaleColor(
    Scale Alpha=&FloatFromCurve(
     FloatCurve=DataInterface() {propertyValues {Curve {keys [
         {interpMode=RCIM_Cubic, tangentMode=RCTM_User, time=0.01, value=0}
         {interpMode=RCIM_Cubic, tangentMode=RCTM_User, time=0.1, value=1}
         {arriveTangent=0.0281, interpMode=RCIM_Cubic, leaveTangent=0.0281, tangentMode=RCTM_User, time=0.5, value=1}
         <at 1 0 RCIM_Cubic>
        ]}}}
     Scale Curve=&FloatFromCurve(
      FloatCurve=DataInterface() {propertyValues {Curve {keys [
          {interpMode=RCIM_Cubic, tangentMode=RCTM_User, time=0.025, value=0}
          <at 0.1 1>
          {interpMode=RCIM_Cubic, tangentMode=RCTM_User, time=0.9, value=1}
          {tangentMode=RCTM_User, time=0.975, value=0}
         ]}}}
      CurveIndex=$System.NormalizedLoopAge
      Scale Curve=&FloatFromCurve(
       FloatCurve=DataInterface() {propertyValues {Curve {keys [{interpMode=RCIM_Cubic, tangentMode=RCTM_User, time=0.5, value=1}, {interpMode=RCIM_Cubic, tangentMode=RCTM_User, time=1, value=0}]}}}
       CurveIndex=/Game/Effects/Niagara/DynamicInputs/Math/FN_NormalizedParticleCount()
      )
     )
    )
    Scale RGB=&Multiply_VectorByFloat(
     Vector=<1 1 1>
     Float=&FloatFromCurve(FloatCurve=DataInterface() {propertyValues {Curve {keys [{interpMode=RCIM_Cubic, tangentMode=RCTM_User, time=0, value=1}, {interpMode=RCIM_Cubic, tangentMode=RCTM_User, time=0.75, value=0.3}]}}})
    )
    ScaleA=true
    ScaleRGB=true
   )
   &ScaleRibbonWidth(Ribbon Width Scale=&FloatFromCurve(
     FloatCurve=DataInterface() {propertyValues {Curve {keys [
         <at 0 0 RCIM_Cubic>
         {arriveTangent=1.0811, time=0.05, value=1}
         {arriveTangent=-1.0638, interpMode=RCIM_Cubic, leaveTangent=-1.0638, time=0.95, value=1}
         <at 0.99 0 RCIM_Cubic>
        ]}}}
     CurveIndex=$System.NormalizedLoopAge
     Scale Curve=&FloatFromCurve(FloatCurve=DataInterface() {propertyValues {Curve {keys [
          {arriveTangent=-0.0787, interpMode=RCIM_Cubic, leaveTangent=-0.0787, tangentMode=RCTM_User, time=0, value=0.5}
          {arriveTangent=0.9887, interpMode=RCIM_Cubic, leaveTangent=0.9887, tangentMode=RCTM_User, time=0.075, value=1}
         ]}}})
    ))
   /SetVariables(Particles.RibbonUVDistance:NiagaraFloat=&Multiply_FloatByInt(Float=1.25, Integer=&ReturnParticleIDAttributes())) {isSetParameters=true, module=SetVariables_#1}
   /SetVariables(
    Particles.GlowRibbonWdith:NiagaraFloat=&Multiply_Float(A=$Particles.RibbonWidth, B=1.5)
    Particles.RibbonColor:LinearColor=&MultiplyLinearColor(Linear Color=$Particles.Color, Scale Factor=&MakeVector4_FromFloat(Float=0.75))
   ) {isSetParameters=true, module=SetVariables_#2}
   &KillParticles(Kill Particles=&SetBoolByFloatComparison(A=$System.NormalizedLoopAge, B=0.99, Comparison Type=::NewEnumerator2))
  ]
  rendererClasses [&NiagaraRibbonRendererProperties]
  renderers [{rendererClass=&NiagaraRibbonRendererProperties}]
 }
 PaintTrail_R {
  emitterUpdateScript [
   &EmitterState()
   &SpawnRate(SpawnRate=22, Use Spawn Probability=false) {module=SpawnRate001}
  ]
  particleSpawnScript [
   &InitializeParticle(
    Color=&SelectLinearColorFromArray(
     Array Sampling Mode=::NewEnumerator1
     Direct Array Index=$System.LoopCount
     Direct Array Mode=::NewEnumerator1
     Color Selection Array=$User.BrushColorArray
     Random Seed=?
    )
    Lifetime=$User.Lifetime
    Position Mode=::NewEnumerator0
    Ribbon Width=$User.RibbonWidth
    Ribbon Width Mode=::NewEnumerator1
    Write Material Random=true
   )
   &SkeletalMeshLocation(
    Bone / Socket Index=1
    Bone Sampling Mode=::NewEnumerator4
    Interpolate Mesh Sampling=false
    MaskSpawnGroup=false
    Orientation Sampling=::NewEnumerator1
    Sampled Position Offset Coordinate Space:ENiagaraSkelMeshTransforms=::NewEnumerator0
    Sampled Velocity Coordinate Space:ENiagaraCoordinateSpace=::Local
    Skeletal Mesh=$User.SkeletalMesh
    Use Sampled Mesh Attribute as Ribbon ID=false
   ) {module=SkeletalMeshLocation001}
  ]
  particleUpdateScript [
   &ParticleState()
   &ScaleColor(
    Scale Alpha=&FloatFromCurve(
     FloatCurve=DataInterface() {propertyValues {Curve {keys [
         {interpMode=RCIM_Cubic, tangentMode=RCTM_User, time=0.01, value=0}
         {interpMode=RCIM_Cubic, tangentMode=RCTM_User, time=0.1, value=1}
         {arriveTangent=0.0281, interpMode=RCIM_Cubic, leaveTangent=0.0281, tangentMode=RCTM_User, time=0.5, value=1}
         <at 1 0 RCIM_Cubic>
        ]}}}
     Scale Curve=&FloatFromCurve(
      FloatCurve=DataInterface() {propertyValues {Curve {keys [
          {interpMode=RCIM_Cubic, tangentMode=RCTM_User, time=0.025, value=0}
          <at 0.1 1>
          {interpMode=RCIM_Cubic, tangentMode=RCTM_User, time=0.9, value=1}
          {tangentMode=RCTM_User, time=0.975, value=0}
         ]}}}
      CurveIndex=$System.NormalizedLoopAge
      Scale Curve=&FloatFromCurve(
       FloatCurve=DataInterface() {propertyValues {Curve {keys [{interpMode=RCIM_Cubic, tangentMode=RCTM_User, time=0.5, value=1}, {interpMode=RCIM_Cubic, tangentMode=RCTM_User, time=1, value=0}]}}}
       CurveIndex=&FN_NormalizedParticleCount()
      )
     )
    )
    Scale RGB=&Multiply_VectorByFloat(
     Vector=<1 1 1>
     Float=&FloatFromCurve(FloatCurve=DataInterface() {propertyValues {Curve {keys [{interpMode=RCIM_Cubic, tangentMode=RCTM_User, time=0, value=1}, {interpMode=RCIM_Cubic, tangentMode=RCTM_User, time=0.75, value=0.3}]}}})
    )
    ScaleA=true
    ScaleRGB=true
   )
   &ScaleRibbonWidth(Ribbon Width Scale=&FloatFromCurve(
     FloatCurve=DataInterface() {propertyValues {Curve {keys [
         <at 0 0 RCIM_Cubic>
         {arriveTangent=1.0811, time=0.05, value=1}
         {arriveTangent=-1.0638, interpMode=RCIM_Cubic, leaveTangent=-1.0638, time=0.95, value=1}
         <at 0.99 0 RCIM_Cubic>
        ]}}}
     CurveIndex=$System.NormalizedLoopAge
     Scale Curve=&FloatFromCurve(FloatCurve=DataInterface() {propertyValues {Curve {keys [
          {arriveTangent=9.0725, interpMode=RCIM_Cubic, leaveTangent=9.0725, tangentMode=RCTM_User, time=0, value=0.5}
          {arriveTangent=-2.5249, interpMode=RCIM_Cubic, leaveTangent=-2.5249, tangentMode=RCTM_User, time=0.075, value=1}
         ]}}})
    ))
   /SetVariables(Particles.RibbonUVDistance:NiagaraFloat=&Multiply_FloatByInt(Float=1.25, Integer=&ReturnParticleIDAttributes())) {isSetParameters=true, module=SetVariables_#1}
   /SetVariables(
    Particles.GlowRibbonWdith:NiagaraFloat=&Multiply_Float(A=$Particles.RibbonWidth, B=1.5)
    Particles.RibbonColor:LinearColor=&MultiplyLinearColor(Linear Color=$Particles.Color, Scale Factor=&MakeVector4_FromFloat(Float=0.75))
   ) {isSetParameters=true, module=SetVariables_#2}
   &KillParticles(Kill Particles=&SetBoolByFloatComparison(A=$System.NormalizedLoopAge, B=0.99, Comparison Type=::NewEnumerator2))
  ]
  rendererClasses [&NiagaraRibbonRendererProperties]
  renderers [{rendererClass=&NiagaraRibbonRendererProperties}]
 }
}
summary {
 systemName=NS_Emote_Dazzle_Trail
 emitters [
  {emitterName=Embers, simTarget=GPUComputeSim, rendererClasses [&NiagaraSpriteRendererProperties]}
  {emitterName=Glow, rendererClasses [&NiagaraSpriteRendererProperties]}
  {emitterName=PaintTrail_L, rendererClasses [&NiagaraRibbonRendererProperties]}
  {emitterName=PaintTrail_R, rendererClasses [&NiagaraRibbonRendererProperties]}
 ]
 userVariables [
  User.BrushColorArray {
   type=NiagaraDataInterfaceArrayColor
   description=""
   defaultValue {
    struct=NiagaraExt_VariableValue_DataInterface
    value {
     dataInterface=&self:NiagaraDataInterfaceArrayColor_2
     dataInterfaceClass=&NiagaraDataInterfaceArrayColor
    }
   }
  }
  User.Lifetime {type=NiagaraFloat, defaultValue=1.5, description=""}
  User.RibbonIDOffset {type=NiagaraInt32}
  User.RibbonWidth {type=NiagaraFloat, defaultValue=8, description=""}
  User.SkeletalMesh {
   type=NiagaraDataInterfaceSkeletalMesh
   defaultValue {
    struct=NiagaraExt_VariableValue_DataInterface
    value {dataInterface=&self:NiagaraDataInterfaceSkeletalMesh_0}
   }
  }
 ]
}
